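# Sound cue that plays musical ambient one-shots for the Nitro Wasteland POI.
# A single Random node picks with equal weight among 29 wave players, covering lettered variants of thirteen numbered Wasteland one-shots.
# Output is quiet, uses the gameplay music sound class and a no-reverb musical one-shot attenuation.
# A concurrency override allows one instance at a time and blocks new starts while one plays.
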SoundCue /Game/Sounds/Helios/POI/NitroWasteland/Spots/Musical/FN_S30_AmbientOneShots_Nitro_Cue
root=Random_1
nodes [
 {
  id=Random_1
  children [
   WavePlayer_4
   WavePlayer_5
   WavePlayer_6
   WavePlayer_10
   WavePlayer_11
   WavePlayer_18
   WavePlayer_19
   WavePlayer_20
   WavePlayer_21
   WavePlayer_22
   WavePlayer_23
   WavePlayer_24
   WavePlayer_25
   WavePlayer_26
   WavePlayer_27
   WavePlayer_28
   WavePlayer_29
   WavePlayer_30
   WavePlayer_31
   WavePlayer_32
   WavePlayer_33
   WavePlayer_34
   WavePlayer_35
   WavePlayer_36
   WavePlayer_37
   WavePlayer_38
   WavePlayer_39
   WavePlayer_40
   WavePlayer_41
  ]
  properties {weights [1,1,1,1,1,1,1,1,1,1,1,1,1,1,1,1,1,1,1,1,1,1,1,1,1,1,1,1,1]}
 }
 {id=WavePlayer_10, properties {soundWaveAssetPtr=&/Amb_Music_Wasteland_OneShot_12a}}
 {id=WavePlayer_11, properties {soundWaveAssetPtr=&/Amb_Music_Wasteland_OneShot_11b}}
 {id=WavePlayer_18, properties {soundWaveAssetPtr=&/Amb_Music_Wasteland_OneShot_11a}}
 {id=WavePlayer_19, properties {soundWaveAssetPtr=&/Amb_Music_Wasteland_OneShot_10b}}
 {id=WavePlayer_20, properties {soundWaveAssetPtr=&/Amb_Music_Wasteland_OneShot_10a}}
 {id=WavePlayer_21, properties {soundWaveAssetPtr=&/Amb_Music_Wasteland_OneShot_09b}}
 {id=WavePlayer_22, properties {soundWaveAssetPtr=&/Amb_Music_Wasteland_OneShot_09a}}
 {id=WavePlayer_23, properties {soundWaveAssetPtr=&/Amb_Music_Wasteland_OneShot_08b}}
 {id=WavePlayer_24, properties {soundWaveAssetPtr=&/Amb_Music_Wasteland_OneShot_08a}}
 {id=WavePlayer_25, properties {soundWaveAssetPtr=&/Amb_Music_Wasteland_OneShot_07b}}
 {id=WavePlayer_26, properties {soundWaveAssetPtr=&/Amb_Music_Wasteland_OneShot_07a}}
 {id=WavePlayer_27, properties {soundWaveAssetPtr=&/Amb_Music_Wasteland_OneShot_06b}}
 {id=WavePlayer_28, properties {soundWaveAssetPtr=&/Amb_Music_Wasteland_OneShot_06a}}
 {id=WavePlayer_29, properties {soundWaveAssetPtr=&/Amb_Music_Wasteland_OneShot_05b}}
 {id=WavePlayer_30, properties {soundWaveAssetPtr=&/Amb_Music_Wasteland_OneShot_05a}}
 {id=WavePlayer_31, properties {soundWaveAssetPtr=&/Amb_Music_Wasteland_OneShot_04b}}
 {id=WavePlayer_32, properties {soundWaveAssetPtr=&/Amb_Music_Wasteland_OneShot_04a}}
 {id=WavePlayer_33, properties {soundWaveAssetPtr=&/Amb_Music_Wasteland_OneShot_03c}}
 {id=WavePlayer_34, properties {soundWaveAssetPtr=&/Amb_Music_Wasteland_OneShot_03b}}
 {id=WavePlayer_35, properties {soundWaveAssetPtr=&/Amb_Music_Wasteland_OneShot_03a}}
 {id=WavePlayer_36, properties {soundWaveAssetPtr=&/Amb_Music_Wasteland_OneShot_02c}}
 {id=WavePlayer_37, properties {soundWaveAssetPtr=&/Amb_Music_Wasteland_OneShot_02b}}
 {id=WavePlayer_38, properties {soundWaveAssetPtr=&/Amb_Music_Wasteland_OneShot_02a}}
 {id=WavePlayer_39, properties {soundWaveAssetPtr=&/Amb_Music_Wasteland_OneShot_01c}}
 {id=WavePlayer_4, properties {soundWaveAssetPtr=&/Amb_Music_Wasteland_OneShot_13b}}
 {id=WavePlayer_40, properties {soundWaveAssetPtr=&/Amb_Music_Wasteland_OneShot_01b}}
 {id=WavePlayer_41, properties {soundWaveAssetPtr=&/Amb_Music_Wasteland_OneShot_01a}}
 {id=WavePlayer_5, properties {soundWaveAssetPtr=&/Amb_Music_Wasteland_OneShot_13a}}
 {id=WavePlayer_6, properties {soundWaveAssetPtr=&/Amb_Music_Wasteland_OneShot_12b}}
]
properties {
 attenuationSettings=/Game/Sounds/Attenuation/Fort_Ambient_Athena_Musical_Oneshot_NoVerb_Att
 bOverrideConcurrency=true
 maxDistance=2097152
 soundClassObject=/Game/Sounds/FortSoundClassesAndModes/Fort_Music_Gameplay
 volumeMultiplier=0.15
 concurrencyOverrides {maxCount=1, resolutionRule=PreventNew, voiceStealReleaseTime=2}
}
